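; Gameplay ability for a Battle Pass S20 quest that credits the player for surviving a high fall.
; On activation it records the pawn's starting Z and starts waits for falling-damage tag addition and falling movement-mode tag removal.
; The damage tag ends the ability without credit.
; When falling stops, a living pawn that dropped at least Height Check sends the Survive.HighFall stat event to its quest manager, and the ability ends.
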(event Ability|EventActivateAbilityFromEvent
  (params (EventData :type Gameplay%20Event%20Data%20Structure%20%28by%20ref%29))
  (|CallParentFunction EventData)
  (Variables|Default|SetActor%20Initial%20Height (.z (Transformation|GetActorLocation (Variables|GAVars|Data|GetOwningPlayerPawn))))
  (Ability|Tasks|WaitGameplayTagAdd "(TagName=\"Event.Damage.Received.Falling\")"
    (:then
      (Ability|Tasks|WaitGameplayTagRemove "(TagName=\"MovementMode.Falling\")"
        (:Removed
          (if (FortPawn|IsDead (Variables|GAVars|Data|GetOwningPlayerPawn))
            _
            (elif (>= (- (Variables|Default|GetActor%20Initial%20Height) (Variables|Default|GetHeight%20Check)) (.z (Transformation|GetActorLocation (Variables|GAVars|Data|GetOwningPlayerPawn))))
              (Quest|SendComplexCustomStatEvent (Variables|GAVars|Data|GetOwningPlayerQuestManager) 0 "(GameplayTags=((TagName=\"Athena.Quests.Survive.HighFall\")))" "(GameplayTags=)" 1)
              (else
                _)))
          (Ability|EndAbility))))
    (:Added
      (Ability|EndAbility))))
